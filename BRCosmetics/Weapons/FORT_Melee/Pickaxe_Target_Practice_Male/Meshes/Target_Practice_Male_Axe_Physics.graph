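# Physics asset for the Target Practice Male pickaxe, a cosmetic melee weapon, defining bodies and joints for its dangling arrow and tape bones.
# A long capsule covers the two-handed handle, and small zero-length capsules act as spheres on the dynamic bones.
# The first arrow bone swings almost like a hinge off the handle, with the second chained to it on looser limits.
# The tape hangs separately from the handle with moderate swing and little twist.
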PhysicsAsset /BRCosmetics/Weapons/FORT_Melee/Pickaxe_Target_Practice_Male/Meshes/Target_Practice_Male_Axe_Physics
bodies {
 Melee_TwoHanded_Handle [{shapeType=Capsule, length=71.3081, radius=5}]
 dyn_arrow_01 [{shapeType=Capsule, length=0, radius=3, center=<8 0 0>}]
 dyn_arrow_02 [{shapeType=Capsule, length=0, radius=3, center=<3 0 0>}]
 dyn_tape [{shapeType=Capsule, length=0, radius=3, center=<6 0 0>}]
}
constraints [
 {bone1Name=dyn_arrow_01, bone2Name=Melee_TwoHanded_Handle, swing1LimitDegrees=90, swing2LimitDegrees=2, twistLimitDegrees=2}
 {bone1Name=dyn_arrow_02, bone2Name=dyn_arrow_01, swing1LimitDegrees=90, swing2LimitDegrees=90, twistLimitDegrees=20}
 {bone1Name=dyn_tape, bone2Name=Melee_TwoHanded_Handle, swing1LimitDegrees=60, swing2LimitDegrees=60, twistLimitDegrees=5}
]
